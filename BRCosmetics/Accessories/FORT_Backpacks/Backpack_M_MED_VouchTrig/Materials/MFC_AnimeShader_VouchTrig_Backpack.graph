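# Material function that supplies the map and parameter inputs for the anime shader on the VouchTrig backpack cosmetic.
# Static switches choose textures or fallbacks for lit/shaded color, normal, DFL mask and color, STT (sheen/threshold/thickness), SRM and specular color.
# Lit and shaded colors pass through shared cinematic tint functions, with optional dynamic darkening for shaded and DFL colors.
# It also outputs rim-light color, an FX mask and Fresnel-weighted emissive.
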MaterialFunction /BRCosmetics/Accessories/FORT_Backpacks/Backpack_M_MED_VouchTrig/Materials/MFC_AnimeShader_VouchTrig_Backpack
graph {nodes [
  n1 = VectorParameter() {desc="RGB=Color", group=01_Anime Maps, parameterName=StyleRimLight_Color, sortPriority=22, defaultValue=<rgba 1 1 1 1>}
  AppendVector(A=@n1.RGB, B=@n1.RGB)
  n2 = TextureSampleParameter2D(UVs=NamedRerouteUsage()) {group=01_Anime Maps, parameterName=Color_Lit_Map, sortPriority=10, texture=&T_WhitesRGB}
  n3 = StaticSwitchParameter(True=@n2.RGB, False=VectorParameter() {group=01_Anime Maps, parameterName=Color_Lit, sortPriority=10, defaultValue=<rgba 0.9 0.9 0.9 0>}.RGB) {group=01_Anime Maps, parameterName=UseColorLitMap, sortPriority=0}
  FunctionOutput(None=NamedRerouteDeclaration(None=MaterialFunctionCall!(LitBaseColorIn=@n3) {materialFunction=/Game/Packages/Fortress_SharedMaterials/Cinematics/Anime/MF_CINE_LitDiffuseTint}.Result) {name=Color_Lit, nodeColor=<rgba 0 0.4609 1 1>}) {outputName=Color_Lit, sortPriority=11}
  n4 = TextureSampleParameter2D(UVs=NamedRerouteUsage()) {group=01_Anime Maps, parameterName=Color_Shaded_Map, sortPriority=11, texture=/Game/Global/Textures/Default/Blanks/T_MidgreysRGB}
  n5 = StaticSwitchParameter(
   True=@n4.RGB
   False=Multiply(
    A=NamedRerouteUsage()
    B=ScalarParameter() {defaultValue=0.5, group=01_Anime Maps, parameterName=Dynamic_ColorShaded_Darkening, sliderMax=1, sortPriority=11}
   )
  ) {group=01_Anime Maps, parameterName=UseColorShadedMap, sortPriority=2}
  n6 = VectorParameter() {group=01_Anime Maps, parameterName=Color_Shaded, sortPriority=11, defaultValue=<rgba 0.18 0.18 0.18 0>}
  FunctionOutput(None=MaterialFunctionCall!(ShadedBaseColorIn=StaticSwitchParameter(True=@n5, False=@n6.RGB) {group=01_Anime Maps, parameterName=UseDynamicColorShaded, sortPriority=1}) {materialFunction=/Game/Packages/Fortress_SharedMaterials/Cinematics/Anime/MF_CINE_ShadedDiffuseTint}.Result) {outputName=Color_Shaded, sortPriority=12}
  FunctionOutput(None=@n1.RGB) {outputName=StyleRimLight_Color, sortPriority=17}
  n7 = TextureSampleParameter2D() {
   group=01_Anime Maps
   parameterName=DFL_Map
   samplerType=SAMPLERTYPE_Alpha
   sortPriority=14
   texture=/Game/Global/Textures/Default/Blanks/T_BlackAlpha
  }
  FunctionOutput(None=StaticSwitchParameter(True=@n7.R, False=Constant()) {group=01_Anime Maps, parameterName=UseDFLMap, sortPriority=4}) {outputName=DFL, sortPriority=3}
  n8 = TextureSampleParameter2D() {
   group=01_Anime Maps
   parameterName=Normals
   samplerType=SAMPLERTYPE_Normal
   sortPriority=13
   texture=&FlatNormalBlue
  }
  FunctionOutput(None=StaticSwitchParameter(True=@n8.RGB, False=Constant3Vector() {constant=<rgba 0 0 1 0>}) {group=01_Anime Maps, parameterName=UseNormalMap, sortPriority=3}) {outputName=Normal, sortPriority=2}
  n9 = TextureSampleParameter2D() {
   desc="R=Sheen \r\nG=Threshold\r\nB=Thickness"
   group=01_Anime Maps
   parameterName=STT_Map
   samplerType=SAMPLERTYPE_Masks
   sortPriority=12
   texture=/Game/Global/Textures/Default/Blanks/T_FN_Anime_STT
   channelNames {a="", b=Thickness, g="Threshold ", r=Sheen}
  }
  n10 = ChannelMaskParameter(None=@n9.RGB) {group=01_Anime Maps, parameterName=STT_Channel_Sheen, sortPriority=21}
  n11 = ChannelMaskParameter(None=@n9.RGB) {group=01_Anime Maps, maskChannel=Green, parameterName=STT_Channel_Threshold, sortPriority=22, defaultValue=<rgba 0 1 0 0>}
  n12 = ChannelMaskParameter(None=@n9.RGB) {group=01_Anime Maps, maskChannel=Blue, parameterName=STT_Channel_Thickness, sortPriority=23, defaultValue=<rgba 0 0 1 0>}
  n13 = ScalarParameter() {group=01_Anime Maps, parameterName=STT_Sheen, sliderMax=1, sortPriority=21}
  n14 = ScalarParameter() {defaultValue=0.5, group=01_Anime Maps, parameterName=STT_Threshold, sliderMax=1, sortPriority=22}
  n15 = ScalarParameter() {defaultValue=1, group=01_Anime Maps, parameterName=STT_Thickness, sliderMax=1, sortPriority=23}
  FunctionOutput(None=StaticSwitchParameter(
    True=MaterialFunctionCall(X=@n10.R, Y=@n11.R, Z=@n12.R) {materialFunction=&MakeFloat3}.Result
    False=MaterialFunctionCall(X=@n13, Y=@n14, Z=@n15) {materialFunction=&MakeFloat3}.Result
   ) {group=01_Anime Maps, parameterName=UseSTTMap, sortPriority=3}) {outputName=STT, sortPriority=1}
  n16 = TextureSampleParameter2D() {group=01_Anime Maps, parameterName=DFL_Color_Map, sortPriority=18, texture=&T_BlacksRGB}
  n17 = StaticSwitchParameter(
   True=@n16.RGB
   False=Multiply(A=NamedRerouteUsage(), B=ScalarParameter() {group=01_Anime Maps, parameterName=Dynamic_DFLColor_Darkening, sliderMax=1, sortPriority=19})
  ) {group=01_Anime Maps, parameterName=UseDFL_ColorMap, sortPriority=9}
  n18 = VectorParameter() {group=01_Anime Maps, parameterName=DFLine_Color, sortPriority=20, defaultValue=<rgba 0 0 0 1>}
  FunctionOutput(None=Saturate(None=StaticSwitchParameter(True=@n17, False=@n18.RGB) {group=01_Anime Maps, parameterName=UseDynamicDFLColor, sortPriority=8})) {outputName=DFL_Color, sortPriority=15}
  n19 = TextureSampleParameter2D() {group=01_Anime Maps, parameterName=Emissive_Map, sortPriority=17, texture=&T_BlacksRGB}
  n20 = Multiply(A=@n19.RGB, B=ScalarParameter() {defaultValue=1, group=02_Anime Parameters, parameterName=Emissive_Intensity, sliderMax=10})
  n21 = CameraVectorWS()
  n22 = PixelNormalWS()
  n23 = Saturate(None=Power(Base=DotProduct(A=@n21, B=@n22), Exp=ScalarParameter() {group=02_Anime Parameters, parameterName=Emissive_FresnelExponent, sliderMax=8}))
  FunctionOutput(None=StaticSwitchParameter(True=Multiply(A=@n20, B=@n23), False=Constant3Vector()) {group=01_Anime Maps, parameterName=UseEmissiveMap, sortPriority=7}) {outputName=Emissive, sortPriority=14}
  n24 = TextureSampleParameter2D() {
   desc="R=Spec\r\nG=Rough\r\nB=Metallic"
   group=01_Anime Maps
   parameterName=SRM_Map
   samplerType=SAMPLERTYPE_Masks
   sortPriority=15
   texture=&T_Fortnite_Default_S
   channelNames {a="", b=Metallic, g=Roughness, r=Specular}
  }
  n25 = ScalarParameter() {defaultValue=0.5, group=01_Anime Maps, parameterName=SRM_Specular, sliderMax=1, sortPriority=15}
  n26 = ScalarParameter() {defaultValue=0.5, group=01_Anime Maps, parameterName=SRM_Roughness, sliderMax=1, sortPriority=15}
  n27 = ScalarParameter() {group=01_Anime Maps, parameterName=SRM_Metallic, sliderMax=1, sortPriority=15}
  FunctionOutput(None=StaticSwitchParameter(
    True=@n24.RGB
    False=MaterialFunctionCall(X=@n25, Y=@n26, Z=@n27) {materialFunction=&MakeFloat3}.Result
   ) {group=01_Anime Maps, parameterName=UseSRMMap, sortPriority=5}) {outputName=SRM, sortPriority=13}
  n28 = TextureSampleParameter2D() {
   group=01_Anime Maps
   parameterName=FX_Map
   samplerType=SAMPLERTYPE_Masks
   sortPriority=16
   texture=&T_BlackMask
  }
  FunctionOutput(None=@n28.RGBA) {outputName=FX, sortPriority=4}
  FunctionOutput() {outputName=--------, sortPriority=10}
  n29 = TextureSampleParameter2D() {group=01_Anime Maps, parameterName=SpecularColor_Map, sortPriority=19, texture=&T_BlacksRGB}
  n30 = ScalarParameter() {defaultValue=1, group=01_Anime Maps, parameterName=SpecularColor_Opacity, sliderMax=1, sortPriority=19}
  n31 = VectorParameter() {
   desc="RGB=Color, A=Opacity"
   group=01_Anime Maps
   parameterName=Specular_Color
   sortPriority=21
   channelNames {a=Opacity, b=Blue Channel, g=Green Channel, r=Red Channel}
   defaultValue=<rgba 1 1 1 1>
  }
  FunctionOutput(None=StaticSwitchParameter(True=AppendVector(A=@n29.RGB, B=@n30), False=AppendVector(A=@n31, B=@n31)) {group=01_Anime Maps, parameterName=UseSpecularColorMap, sortPriority=9}) {outputName=Specular_Color, sortPriority=16}
  NamedRerouteDeclaration(None=FunctionInput(Preview=TextureCoordinate()) {bUsePreviewValueAsDefault=true, inputName=CustomUV, inputType=FunctionInput_Vector2}) {name=UV Coords, nodeColor=<rgba 0.5 0.5 0 1>}
  NamedRerouteUsage()
  NamedRerouteUsage()
  NamedRerouteUsage()
  NamedRerouteUsage()
  NamedRerouteUsage()
  NamedRerouteUsage()
  NamedRerouteUsage()
  NamedRerouteUsage()
  TextureSampleParameter2D() {
   group=01_Anime Maps
   parameterName=Normal_Map
   samplerType=SAMPLERTYPE_Normal
   sortPriority=13
   texture=&FlatNormalBlue
  }
 ]}
