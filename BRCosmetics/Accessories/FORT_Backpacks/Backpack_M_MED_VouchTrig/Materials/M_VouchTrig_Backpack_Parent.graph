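# Parent material for the VouchTrig backpack, combining anime-style character shading with a masked, animated rotating, wavy UV effect on a cloth shading model.
# A time-driven custom rotator, distorted by panning noise, replaces the base UVs where the FX mask's red channel applies.
# These UVs feed the backpack's anime shader function, then shared Calculations and Post functions, with eye animation applied to the color.
# A second scale-and-rotate UV chain feeds nothing.
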Material /BRCosmetics/Accessories/FORT_Backpacks/Backpack_M_MED_VouchTrig/Materials/M_VouchTrig_Backpack_Parent
graph {
 properties {
  bAutomaticallySetUsageInEditor=false
  bUseMaterialAttributes=true
  bUsedWithClothing=true
  bUsedWithMeshDeformer=true
  bUsedWithMorphTargets=true
  bUsedWithNiagaraMeshParticles=true
  bUsedWithSkeletalMesh=true
  bUsedWithStaticLighting=true
  pixelDepthOffsetMode=PDOM_Legacy
  shadingModel=MSM_Cloth
  usedShadingModels=Cloth
 }
 nodes [
  n1 = TextureCoordinate()
  n2 = MaterialFunctionCall(
   UVs=@n1
   Rotation Center=Constant2Vector() {g=0.592, r=0.199}
   "Rotation Angle (0-1)"=Multiply(A=Time(), B=ScalarParameter() {defaultValue=-0.077, group=VouchTrig, parameterName=VouchTrig_RotationSpeed}) {constB=-0.05}
  ) {materialFunction=&CustomRotator}
  n3 = TextureSample(UVs=Panner() {speedY=0.1}) {
   samplerSource=SSM_Wrap_WorldGroupSettings
   samplerType=SAMPLERTYPE_Normal
   texture=/Game/Characters/Player/Female/Medium/Bodies/F_MED_Bistro_Astronaut/FX/T_GaussianNoise_Normal
  }
  n4 = Multiply(
   A=ComponentMask(None=@n3.RGB) {g=true, r=true}
   B=Multiply(A=ScalarParameter() {defaultValue=0.125, group=VouchTrig, parameterName=VouchTrig_WaveDistortionAmount}) {constB=0.1}
  ) {constB=0.03}
  n5 = TextureSample() {
   samplerSource=SSM_Wrap_WorldGroupSettings
   samplerType=SAMPLERTYPE_Masks
   texture=/BRCosmetics/Accessories/FORT_Backpacks/Backpack_M_MED_VouchTrig/Textures/T_VouchTrig_TA_Mask_Backpack_FX
  }
  n6 = MaterialFunctionCall!(CustomUV=LinearInterpolate(A=TextureCoordinate(), B=Add(A=@n2.Rotated Values, B=@n4), Alpha=@n5.R)) {materialFunction=&/MFC_AnimeShader_VouchTrig_Backpack}
  n7 = MaterialFunctionCall!(CustomLightVector, OverrideNormal_WS, STT=@n6.STT, Normal_Map=@n6.STT, DFL=@n6.STT, FX=@n6.STT) {materialFunction=/Game/Global/Materials/Character_Parent_2023/Anime_Parent_2023/MaterialFunctions/MFC_AnimeShader_2023_Calculations}
  n8 = VertexColor()
  n9 = TextureSampleParameter2D() {
   group=01_EyeAnimation_Parameters
   parameterName=EyeMask
   samplerSource=SSM_Wrap_WorldGroupSettings
   samplerType=SAMPLERTYPE_Masks
   sortPriority=-32
   texture=&T_WhiteMask
  }
  n10 = StaticSwitchParameter(True=AppendVector(A=@n8, B=@n8), False=@n9.RGBA) {group=01_EyeAnimation_Parameters, parameterName=UseVertexColorForEyeMask, sortPriority=-33}
  n11 = TextureCoordinate() {coordinateIndex=1}
  n12 = Sign(None=Subtract(A=ComponentMask(None=@n11) {r=true}) {constB=0.0001})
  n13 = MaterialFunctionCall!(
   Color Lit=@n6.Color_Lit
   Color Shaded=@n6.Color_Lit
   Use UV2=StaticBoolParameter() {defaultValue=true, group=01_EyeAnimation_Parameters, parameterName=Use UV2 for Eye Animation, sortPriority=-30}
   Mask=StaticSwitchParameter(
    True=ChannelMaskParameter(None=@n10) {group=01_EyeAnimation_Parameters, parameterName=EyeMaskChannel, sortPriority=-32}.R
    False=Saturate(None=@n12)
   ) {group=01_EyeAnimation_Parameters, parameterName=UseEyeMask, sortPriority=-33}
  ) {materialFunction=/Game/Global/Materials/Character_Parent_2023/Anime_Parent_2023/MaterialFunctions/MF_AnimeEyeAnimation_v2}
  MaterialFunctionCall!(
   Shading_Gradient=@n7.Shading_Gradient
   Threshold_Gradient=@n7.Shading_Gradient
   Specular_Gradient=@n7.Shading_Gradient
   Sheen_Mask=@n7.Shading_Gradient
   Thickness=@n7.Shading_Gradient
   DFL_Mask=@n7.Shading_Gradient
   FX=@n7.Shading_Gradient
   Metallic_Gradient=@n7.Shading_Gradient
   RimLight_Gradient=@n7.Shading_Gradient
   NormalBuffer=@n7.Shading_Gradient
   --------
   CINE_LightSources=@n7.Shading_Gradient
   CINE_LightColor=@n7.Shading_Gradient
   ---------
   Color_Lit=@n13.Lit
   Color_Shaded=@n13.Lit
   SRM=@n6.SRM
   Emissive=@n6.SRM
   DFL_Color=@n6.SRM
   Specular_Color=@n6.SRM
   RimLight_Color=@n6.SRM
  ) {materialFunction=/Game/Global/Materials/Character_Parent_2023/Anime_Parent_2023/MaterialFunctions/MFC_AnimeShader_2023_Post}
  n14 = Constant() {r=1}
  n15 = Constant2Vector() {g=-0.1, r=0.3}
  Rotator(
   Coordinate=Add(
    A=MaterialFunctionCall(UVs=@n1, Texture Scale=@n14) {materialFunction=&ScaleUVsByCenter}.UVs
    B=Divide(A=@n15, B=@n14)
   ) {constB=0.5}
   Time=Multiply(A=Time()) {constB=2}
  )
 ]
}
