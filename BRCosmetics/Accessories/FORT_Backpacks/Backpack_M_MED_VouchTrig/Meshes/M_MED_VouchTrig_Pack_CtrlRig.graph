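# Control Rig for the MED_VouchTrig backpack, which keeps the dyn_Bowl_K bone, parent of four rigged ducks, level with the world.
# It converts an identity world transform into rig space, splits its rotation into Euler angles, and rebuilds it from X and Y only, with Z zeroed.
# The result sets the bone's global rotation, cancelling the component's tilt while ignoring its yaw.
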ControlRigBlueprint /BRCosmetics/Accessories/FORT_Backpacks/Backpack_M_MED_VouchTrig/Meshes/M_MED_VouchTrig_Pack_CtrlRig
graphs [
 RigVMFunctionLibrary {links [], nodes []}
 RigVMModel {nodes [
   RigUnit_BeginExecution = &RigUnit_BeginExecution()
   From World = /Script/ControlRig.RigUnit_ToRigSpace_Transform(Value=(Rotation=(X=0,Y=0,Z=0,W=1),Translation=(X=0,Y=0,Z=0),Scale3D=(X=1,Y=1,Z=1))) {is=struct RigVMUnitNode, template="From World::Execute(in Value,out Global)"}
   RigVMFunction_MathQuaternionToEuler = &RigVMFunction_MathQuaternionToEuler(RotationOrder=ZYX, Value=@"From World".Global.Rotation)
   RigVMFunction_MathQuaternionFromEuler = &RigVMFunction_MathQuaternionFromEuler(Euler=(X=0,Y=0,Z=0), Euler.X=@RigVMFunction_MathQuaternionToEuler.Result.X, Euler.Y=@RigVMFunction_MathQuaternionToEuler.Result.Y)
   Set Transform = &RigUnit_SetRotation(
    Item=(Type=Bone,Name="dyn_Bowl_K")
    Space=GlobalSpace
    Weight=1
    bInitial=False
    ExecutePin=@RigUnit_BeginExecution.ExecutePin
    Value=@RigVMFunction_MathQuaternionFromEuler.Result
   )
  ]}
]
hierarchy [
 root {loc=<0 0 0>, rot=<0 0 0 1>}
 pack {parent=root, loc=<0 -0.7316 0>, rot=<0 0 0 1>}
 dyn_Bowl_K {parent=pack, loc=<0 -16.3589 -11.0893>, rot=<0 0 0.7071 -0.7071>}
 dyn_Bowl_K1 {parent=dyn_Bowl_K, loc=<0 0 0>, rot=<0 0 0 1>}
 dyn_Ducks {parent=dyn_Bowl_K1, loc=<0 0 9.8964>, rot=<0.5 -0.5 0.5 0.5>}
 dyn_Duck1 {parent=dyn_Ducks, loc=<-0.7584 -0.9157 -9.5048>, rot=<0.048 -0.0396 0.0019 -0.9981>}
 dyn_D1_Tail {parent=dyn_Duck1, loc=<3.7674 -5.3589 -0.2147>, rot=<0.9356 -0.3475 0.0202 -0.0588>}
 dyn_D1_Head {parent=dyn_Duck1, loc=<6.1483 2.8561 0.7681>, rot=<0.9699 0.2353 0.0497 -0.0374>}
 dyn_D1_Beak_up {parent=dyn_D1_Head, loc=<2.8017 -2.9117 0>, rot=<0.6031 -0.3692 0.3692 -0.6031>}
 dyn_D1_Beak_dn {parent=dyn_D1_Head, loc=<2.8017 -2.9117 0>, rot=<-0.6031 0.3692 0.3692 -0.6031>}
 dyn_D1_Wing_l {parent=dyn_Duck1, loc=<4.2479 1.4799 4.0597>, rot=<0.3975 0.0202 0.7522 -0.5252>}
 dyn_D1_Wing_r {parent=dyn_Duck1, loc=<4.8077 2.1611 -2.9886>, rot=<0.4846 -0.7131 -0.0508 -0.5041>}
 dyn_Duck2 {parent=dyn_Ducks, loc=<-0.8221 9.4637 -0.9926>, rot=<0.7071 0 0 -0.7071>}
 dyn_D2_Tail {parent=dyn_Duck2, loc=<3.7386 5.3833 -0.0039>, rot=<0 0.0001 -0.3499 -0.9368>}
 dyn_D2_Head {parent=dyn_Duck2, loc=<6.1898 -2.8696 -0.0037>, rot=<0 0 0.2339 -0.9723>}
 dyn_D2_Beak_up {parent=dyn_D2_Head, loc=<2.8017 -2.9117 0>, rot=<0.6031 -0.3692 0.3692 -0.6031>}
 dyn_D2_Beak_dn {parent=dyn_D2_Head, loc=<2.8017 -2.9117 0>, rot=<-0.6031 0.3692 0.3692 -0.6031>}
 dyn_D2_Wing_l {parent=dyn_Duck2, loc=<4.556 -1.8289 3.5516>, rot=<-0.4518 0.0347 -0.735 -0.5044>}
 dyn_D2_Wing_r {parent=dyn_Duck2, loc=<4.556 -1.8289 -3.5516>, rot=<-0.5044 -0.735 0.0347 -0.4518>}
 dyn_Duck3 {parent=dyn_Ducks, loc=<-0.8221 0.9926 9.4637>, rot=<0 0 0 1>}
 dyn_D3_Tail {parent=dyn_Duck3, loc=<3.7386 5.3833 -0.0039>, rot=<0 0.0001 -0.3499 -0.9368>}
 dyn_D3_Head {parent=dyn_Duck3, loc=<6.1898 -2.8696 -0.0037>, rot=<0 0 0.2339 -0.9723>}
 dyn_D3_Beak_up {parent=dyn_D3_Head, loc=<2.8017 -2.9117 0>, rot=<0.6031 -0.3692 0.3692 -0.6031>}
 dyn_D3_Beak_dn {parent=dyn_D3_Head, loc=<2.8017 -2.9117 0>, rot=<-0.6031 0.3692 0.3692 -0.6031>}
 dyn_D3_Wing_l {parent=dyn_Duck3, loc=<4.556 -1.8289 3.5516>, rot=<-0.4518 0.0347 -0.735 -0.5044>}
 dyn_D3_Wing_r {parent=dyn_Duck3, loc=<4.556 -1.8289 -3.5516>, rot=<-0.5044 -0.735 0.0347 -0.4518>}
 dyn_Duck4 {parent=dyn_Ducks, loc=<-0.8221 -9.4637 0.9926>, rot=<-0.7071 0 0 -0.7071>}
 dyn_D4_Tail {parent=dyn_Duck4, loc=<3.7386 5.3833 -0.0039>, rot=<0 0.0001 -0.3499 -0.9368>}
 dyn_D4_Head {parent=dyn_Duck4, loc=<6.1898 -2.8696 -0.0037>, rot=<0 0 0.2339 -0.9723>}
 dyn_D4_Beak_up {parent=dyn_D4_Head, loc=<2.8017 -2.9117 0>, rot=<0.6031 -0.3692 0.3692 -0.6031>}
 dyn_D4_Beak_dn {parent=dyn_D4_Head, loc=<2.8017 -2.9117 0>, rot=<-0.6031 0.3692 0.3692 -0.6031>}
 dyn_D4_Wing_l {parent=dyn_Duck4, loc=<4.556 -1.8289 3.5516>, rot=<-0.4518 0.0347 -0.735 -0.5044>}
 dyn_D4_Wing_r {parent=dyn_Duck4, loc=<4.556 -1.8289 -3.5516>, rot=<-0.5044 -0.735 0.0347 -0.4518>}
 dyn_Bowl_all {parent=pack, loc=<0 0 -11>, rot=<0.001 0.7071 0.001 -0.7071>}
]
