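# Animation Blueprint for the VouchTrig backpack, playing idle and elimination reaction animations.
# Front-end and in-game state machines, picked by bIsInFrontEnd, trigger eliminations from a delayed repeating timer (enabled via backpack modifier) or the EnemyPlayerEliminated input event, respectively.
# The pose passes through a Control Rig, then bone modifiers offset the bowl and spin the ducks.
# The event graph accumulates spin from the DuckSpinRate curve and derives gimbal values from spine motion.
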AnimBlueprint /BRCosmetics/Accessories/FORT_Backpacks/Backpack_M_MED_VouchTrig/Meshes/M_MED_VouchTrig_Pack_AnimBP
target_skeleton=&/M_MED_VouchTrig_Pack_Skeleton
graphs [
 AnimGraph {nodes [
   n1 = "Transform (Modify) Bone - Bone: dyn_Bowl_K"!(ComponentPose, Alpha=1, Translation="0,0,0", Rotation="0,0,0") -> (Pose) {
    rotationMode=BMM_Additive
    rotationSpace=BCS_BoneSpace
    alphaBoolBlend {blendOption=Linear}
    alphaScaleBiasClamp {bInterpResult=true}
    boneToModify {boneName=dyn_Bowl_K}
   }
   n2 = "Transform (Modify) Bone - Bone: dyn_Bowl_K"!(ComponentPose=@n1, bAlphaBoolEnabled:bool=True, Translation="0,0,0", Rotation="0,0,0") -> (Pose) {
    alphaInputType=Bool
    rotationMode=BMM_Additive
    rotationSpace=BCS_BoneSpace
    alphaBoolBlend {blendInTime=10, blendOption=QuarticInOut, blendOutTime=10}
    alphaScaleBiasClamp {bInterpResult=true}
    boneToModify {boneName=dyn_Bowl_K}
   }
   "Transform (Modify) Bone - Bone: dyn_Bowl_K"!(ComponentPose=@n2, Alpha:float=1, Translation="0,0,-6", Rotation="0,0,0") -> (Pose) {
    rotationMode=BMM_Additive
    rotationSpace=BCS_BoneSpace
    translationMode=BMM_Additive
    translationSpace=BCS_WorldSpace
    alphaBoolBlend {blendOption=QuarticInOut, blendOutTime=10}
    alphaScaleBiasClamp {bInterpResult=true, interpSpeedDecreasing=5, interpSpeedIncreasing=5}
    boneToModify {boneName=dyn_Bowl_K}
    translation {z=-6}
   }
   AnimGraphNode_StateMachine_0 = Backpack States() {is=title AnimGraphNode_StateMachine}
   AnimGraphNode_StateMachine_1 = FE_States() {is=title AnimGraphNode_StateMachine}
   n3 = Control Rig(Source=Blend Poses by bool(BlendPose_0=@AnimGraphNode_StateMachine_1, BlendPose_1=@AnimGraphNode_StateMachine_0, bActiveValue=Get bIsInFrontEnd())) {controlRigAssetReference {blueprintRigClass=&/M_MED_VouchTrig_Pack_CtrlRig.M_MED_VouchTrig_Pack_CtrlRig_C}}
   n4 = "Transform (Modify) Bone - Bone: dyn_Bowl_K"(ComponentPose=Local To Component(LocalPose=@n3), Alpha=1, Translation="0,0,-9", Rotation="0,0,0") {
    rotationMode=BMM_Additive
    rotationSpace=BCS_BoneSpace
    translationMode=BMM_Additive
    translationSpace=BCS_WorldSpace
    alphaBoolBlend {blendOption=QuarticInOut, blendOutTime=10}
    alphaScaleBiasClamp {bInterpResult=true, interpSpeedDecreasing=5, interpSpeedIncreasing=5}
    boneToModify {boneName=dyn_Bowl_K}
    translation {z=-9}
   }
   n5 = "Transform (Modify) Bone - Bone: dyn_Ducks"(ComponentPose=@n4, Alpha=1, Translation="0,0,0", Rotation=Get Duck Rotation() -> (Duck Rotation:Rotator), Scale="1,1,1") {rotationMode=BMM_Additive, rotationSpace=BCS_BoneSpace, alphaBoolBlend {blendOption=Linear}, boneToModify {boneName=dyn_Ducks}}
   Output Pose(Result=Component To Local(ComponentPose=@n5)) {layerGroup=None}
  ]}
 Backpack States {
  entry=Idle
  path=AnimGraph.AnimGraphNode_StateMachine_0.Backpack States
  ids=id
  nodes [
   AnimStateNode_0 = Idle() {stateEntered {guid=#1}}
   AnimStateNode_4 = Elimination() {is=title AnimStateNode}
   AnimStateTransitionNode_11 = AnimStateTransitionNode() {crossfadeDuration=0.25}
   AnimStateTransitionNode_5 = AnimStateTransitionNode() {bAutomaticRuleBasedOnSequencePlayerInState=true, crossfadeDuration=0.25}
  ]
  links [
   {
    dst=Idle
    rule=AnimGraph.AnimGraphNode_StateMachine_0.Backpack States.AnimStateTransitionNode_5.Transition
    src=Elimination
    via=AnimStateTransitionNode_5
   }
   {
    dst=Elimination
    rule=AnimGraph.AnimGraphNode_StateMachine_0.Backpack States.AnimStateTransitionNode_11.Transition
    src=Idle
    via=AnimStateTransitionNode_11
   }
  ]
 }
 Idle {
  path=AnimGraph.AnimGraphNode_StateMachine_0.Backpack States.AnimStateNode_0.Idle
  nodes [Output Animation Pose(Result=Sequence Player 'VouchTrig_IdleAnim_PACK'() {sequence=/BRCosmetics/Accessories/FORT_Backpacks/Backpack_M_MED_VouchTrig/Animation/VouchTrig_IdleAnim_PACK})]
 }
 Elimination {
  path=AnimGraph.AnimGraphNode_StateMachine_0.Backpack States.AnimStateNode_4.Elimination
  nodes [Output Animation Pose(Result=Sequence Player 'VouchTrig_ElimAnim_PACK'() {bLoopAnimation=false, sequence=/BRCosmetics/Accessories/FORT_Backpacks/Backpack_M_MED_VouchTrig/Animation/VouchTrig_ElimAnim_PACK})]
 }
 Transition {
  path=AnimGraph.AnimGraphNode_StateMachine_0.Backpack States.AnimStateTransitionNode_11.Transition
  nodes [Result(bCanEnterTransition=Get bEnemyPlayerEliminated() -> (bEnemyPlayerEliminated:bool))]
 }
 Transition {path=AnimGraph.AnimGraphNode_StateMachine_0.Backpack States.AnimStateTransitionNode_5.Transition, nodes [Result()]}
 FE_States {
  entry=FE_Idle
  path=AnimGraph.AnimGraphNode_StateMachine_1.FE_States
  ids=id
  nodes [
   AnimStateNode_0 = FE_Idle() {is=title AnimStateNode, stateEntered {guid=#1}}
   AnimStateNode_4 = FE_Elimination() {is=title AnimStateNode, stateEntered {guid=#2, notifyName=Entered FE_Reactive_Anim}}
   AnimStateTransitionNode_11 = AnimStateTransitionNode() {crossfadeDuration=0.25}
   AnimStateTransitionNode_5 = AnimStateTransitionNode() {bAutomaticRuleBasedOnSequencePlayerInState=true, crossfadeDuration=0.25}
  ]
  links [
   {
    dst=FE_Idle
    rule=AnimGraph.AnimGraphNode_StateMachine_1.FE_States.AnimStateTransitionNode_5.Transition
    src=FE_Elimination
    via=AnimStateTransitionNode_5
   }
   {
    dst=FE_Elimination
    rule=AnimGraph.AnimGraphNode_StateMachine_1.FE_States.AnimStateTransitionNode_11.Transition
    src=FE_Idle
    via=AnimStateTransitionNode_11
   }
  ]
 }
 FE_Idle {
  path=AnimGraph.AnimGraphNode_StateMachine_1.FE_States.AnimStateNode_0.FE_Idle
  nodes [Output Animation Pose(Result=Sequence Player 'VouchTrig_IdleAnim_PACK'() {sequence=&VouchTrig_IdleAnim_PACK})]
 }
 FE_Elimination {
  path=AnimGraph.AnimGraphNode_StateMachine_1.FE_States.AnimStateNode_4.FE_Elimination
  nodes [Output Animation Pose(Result=Sequence Player 'VouchTrig_ElimAnim_PACK'() {bLoopAnimation=false, sequence=&VouchTrig_ElimAnim_PACK})]
 }
 Transition {
  path=AnimGraph.AnimGraphNode_StateMachine_1.FE_States.AnimStateTransitionNode_11.Transition
  nodes [Result(bCanEnterTransition=Get Play Reactive FE Anim() -> (Play Reactive FE Anim:bool))]
 }
 Transition {path=AnimGraph.AnimGraphNode_StateMachine_1.FE_States.AnimStateTransitionNode_5.Transition, nodes [Result()]}
 EventGraph {nodes [
   n1 = Event BlueprintUpdateAnimation()
   > Set Delta Time X(Delta Time X:double=@n1.DeltaTimeX) -> (Output_Get:double)
   > n2 = Sequence!(execute) -> (then_0, then_1, then_2:Exec, then_3:Exec)
   n3 = GetOwnerSkeletalMeshComponent()
   n4 = BreakRotator(InRot=BreakTransform(InTransform=GetDeltaTransformFromRefPose(self=@n3, BoneName=spine_05, BaseName=root)).Rotation)
   n5 = float * float(A=@n4.Yaw, B=Get Rot Down() -> (Rot Down:float)) {is=title K2Node_PromotableOperator}
   > Set Gimbal Rot(Gimbal Rot:Rotator=MakeRotator(Roll=float * float(A=@n5, B=0.5), Pitch=float * float(A=@n4.Yaw, B=-0.8))) -> (Output_Get:Rotator)
   > Branch(Condition=NOT Boolean(A=Get bIsCrouching()))
   > Set ADSGimbal(ADSGimbal:Rotator=MakeRotator(Roll=float * float(A=@n4.Yaw, B=-0.6), Pitch=float * float(A=@n4.Yaw, B=0.4))) -> (Output_Get:Rotator)
   n6 = Branch(execute=@n2.then_2, Condition=Get bIsCrouching())
   n7 = Get Gimbal Rot() -> (Gimbal Rot:Rotator)
   > Set TranslateFixNeeded(TranslateFixNeeded:double=Remap Float(Input=BreakRotator(InRot=@n7).Pitch, OldRange=(X=-25,Y=-60), NewRange=(X=0,Y=1), Clamp=true) -> (Output=~)) -> (Output_Get:double)
   Set TranslateFixNeeded(execute=@n6.else, TranslateFixNeeded=0)
   n8 = float * float(A=float * float(A=GetCurveValue(CurveName=DuckSpinRate), B=-0.5), B=Get Delta Time X() -> (Delta Time X:double))
   n9 = Get Duck Rotation Last Tick Value() -> (Duck Rotation Last Tick Value:double)
   n10 = Set Duck Rotation Value(execute=@n2.then_3, Duck Rotation Value:double="% (Float)"!(A:double=float + float(A=@n8, B=@n9), B:double=360) -> (ReturnValue:double)) -> (Output_Get:double)
   > Set Duck Rotation(Duck Rotation:Rotator=MakeRotator(Roll=@n10)) -> (Output_Get:Rotator)
   > Set Duck Rotation Last Tick Value(Duck Rotation Last Tick Value:double=Get Duck Rotation Value() -> (Duck Rotation Value:double)) -> (Output_Get:double)
   Branch(execute=@n2.then_1, Condition=NOT Boolean(A=Get bIsSkyDiving()))
   > Set Rot Down(Rot Down:float=GetAlphaFromReferenceTransformDelta(
     SkeletalMeshComponent="Get Skel mesh "() -> ("Skel mesh ":Skeletal Mesh Component Object Reference)
     BoneName=spine_03
     BaseName="head "
     InputRange=(X=0,Y=-120)
     OutputRange=(X=0,Y=-1)
     TwistAxis="0,0,1"
    )) -> (Output_Get:float)
   > Set Rot Up(Rot Up:float=GetAlphaFromReferenceTransformDelta(SkeletalMeshComponent="Get Skel mesh "(), BoneName=spine_04, BaseName=head, InputRange=(X=15,Y=60), TwistAxis="0,0,1")) -> (Output_Get:float)
   n11 = Event OnCharacterCustomizationLoaded()
   > "Set Skel mesh "("Skel mesh ":Skeletal Mesh Component Object Reference=Get Array Item with EFort Custom Part Type(Array=@n11.CustomMeshParts)) -> (Output_Get:Skeletal Mesh Component Object Reference)
   > n12 = Cast To FortPlayerPawn(Object=TryGetFortPawn())
   > n13 = Cast To B_Athena_PartModifier_Backpack_VouchTrig!(
    execute
    Object=GetModifierForPartType!(self:Fort Player Pawn Object Reference=@n12, PartType:EFortCustomPartType Enum=Backpack) -> (ReturnValue:Custom Character Part Modifier Object Reference)
   ) -> (then, CastFailed, AsB Athena Part Modifier Backpack Vouch Trig:B Athena Part Modifier Backpack Vouch Trig Object Reference) {is=title K2Node_DynamicCast}
   > n14 = Branch(Condition=Get FE_ReactivityEnabled!(self:B Athena Part Modifier Backpack Vouch Trig Object Reference=@n13) -> (FE_ReactivityEnabled:bool))
   > Do Once(Reset=@n16.then)
   > n15 = Delay(Duration=4)
   Set Play Reactive FE Anim(execute=[@n17.then, @n15.then], Play Reactive FE Anim:bool=true) -> (Output_Get:bool)
   n16 = Set Play Reactive FE Anim(execute=@n14.else, Play Reactive FE Anim=false)
   Event BlueprintPostEvaluateAnimation() {is=title K2Node_Event}
   > Set bEnemyPlayerEliminated(bEnemyPlayerEliminated:bool=false) -> (Output_Get:bool)
   AnimNotify_Entered FE_Reactive_Anim() {is=title K2Node_Event}
   > Set Play Reactive FE Anim(Play Reactive FE Anim=false)
   > n17 = Delay(Duration=10)
   n18 = Event BlueprintProcessAnimInputEvent()
   > Branch(Condition=IsNamed(self=@n18.AnimInputEvent, InEventName=aie.EnemyPlayerEliminated))
   > Set bEnemyPlayerEliminated(bEnemyPlayerEliminated=true)
  ]}
]
variables [
 ADSGimbal {type=Rotator}
 bEnemyPlayerEliminated {type=boolean}
 Delta Time X {type=number}
 Duck Rotation {type=Rotator}
 Duck Rotation Last Tick Value {type=number}
 Duck Rotation Value {type=number}
 Gimbal Rot {type=Rotator}
 Play Reactive FE Anim {type=boolean}
 Rot Down {type=number}
 Rot Up {type=number}
 "Skel mesh " {type=SkeletalMeshComponent}
 TranslateFixNeeded {type=number}
]
